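# Animation Blueprint for the AstralLilac backpack cosmetic, adding secondary motion to its zippers, handle and bat.
# Additive root offsets feed Anim Dynamics on zippers and handle, then a label constraint into a cached pose.
# Two Base_Bat Anim Dynamics setups with different limits blend by front-end state.
# Each update derives gravity override, skydiving, parachuting from owner spine yaw, simulation space, and a bat-in-hand flag from pickaxe meta tags.
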AnimBlueprint /BRCosmetics/Accessories/FORT_Backpacks/Backpack_M_MED_AstralLilac/Meshes/M_MED_AstralLilac_Pack_AnimBP
target_skeleton=&/M_MED_AstralLilac_Pack_Skeleton
graphs [
 AnimGraph {nodes [
   n1 = Anim Dynamics - Bone: Base_Bat!(ComponentPose=Local To Component(LocalPose=Use cached pose 'InputPose'()), bAlphaBoolEnabled:bool=True, ExternalForce="0,0,0") -> (Pose) {
    alphaInputType=Bool
    angularSpringConstant=5000
    bGravityOverrideInSimSpace=true
    bLinearSpring=true
    bOverrideLinearDamping=true
    bUseGravityOverride=true
    linearDampingOverride=0.975
    linearSpringConstant=350
    simulationSpace=RootRelative
    alphaBoolBlend {blendOption=Linear}
    boundBone {boneName=Base_Bat}
    componentAppliedLinearAccClamp=<750 750 750>
    componentLinearAccScale {x=0.2, z=0.5}
    componentLinearVelScale=<0.2 0.5 0.5>
    gravityOverride {z=-980}
    physicsBodyDefinitions [{
      boundBone {boneName=Base_Bat}
      boxExtents=<12 12 10>
      constraintSetup {angularLimitsMax=<1 1.5 0>, angularLimitsMin=<-1 -1.5 0>, angularTarget=<0 0 0>, linearAxesMax=<0.1 0 1>, linearAxesMin=<-0.1 0 -1>}
      localJointOffset=<0 0 -1>
     }]
   }
   n2 = Anim Dynamics - Bone: Base_Bat(ComponentPose=Local To Component(LocalPose=Use cached pose 'InputPose'()), bAlphaBoolEnabled=True, ExternalForce="0,0,0") {
    alphaInputType=Bool
    angularSpringConstant=5000
    bGravityOverrideInSimSpace=true
    bLinearSpring=true
    bOverrideLinearDamping=true
    bUseGravityOverride=true
    linearDampingOverride=0.975
    linearSpringConstant=350
    simulationSpace=RootRelative
    alphaBoolBlend {blendOption=Linear}
    boundBone {boneName=Base_Bat}
    componentAppliedLinearAccClamp=<750 750 750>
    componentLinearAccScale {x=0.2, z=0.5}
    componentLinearVelScale=<0.5 0.5 0.5>
    gravityOverride {z=-980}
    physicsBodyDefinitions [{
      boundBone {boneName=Base_Bat}
      boxExtents=<12 12 10>
      constraintSetup {angularLimitsMax=<5 5 0>, angularLimitsMin=<-5 -5 0>, angularTarget=<0 0 0>, linearAxesMax=<1 0 3>, linearAxesMin=<-1 0 -1>}
      localJointOffset=<0 0 -1>
     }]
   }
   Output Pose(Result=Blend Poses by bool(BlendPose_0=Component To Local(ComponentPose=@n1), BlendPose_1=Component To Local(ComponentPose=@n2), bActiveValue=Get bIsInFrontEnd()) {blendType=Linear}) {layerGroup=None}
   n3 = "Transform (Modify) Bone - Bone: root"(ComponentPose=Local To Component(LocalPose=Input Pose()), Translation="0,0,-7", Rotation="0,0,5") {
    rotationMode=BMM_Additive
    rotationSpace=BCS_ParentBoneSpace
    translationMode=BMM_Additive
    translationSpace=BCS_BoneSpace
    alphaBoolBlend {blendOption=Linear}
    boneToModify {boneName=root}
    rotation {roll=5}
    translation {z=-7}
   }
   n4 = "Transform (Modify) Bone - Bone: root"!(ComponentPose=@n3, bAlphaBoolEnabled:bool=True, Translation="0,-1.5,0", Rotation="0,0,7", Scale) -> (Pose) {
    alphaInputType=Bool
    rotationMode=BMM_Additive
    scaleMode=BMM_Additive
    translationMode=BMM_Additive
    alphaBoolBlend {blendOption=Linear}
    boneToModify {boneName=root}
    rotation {roll=7}
    translation {y=-1.5}
   }
   n5 = "Transform (Modify) Bone - Bone: root"!(ComponentPose=@n4, Alpha:float, Translation="0,-18,-7", Rotation="0,0,50", Scale) -> (Pose) {
    rotationMode=BMM_Additive
    scaleMode=BMM_Additive
    translationMode=BMM_Additive
    alphaBoolBlend {blendOption=Linear}
    boneToModify {boneName=root}
    rotation {roll=50}
    translation {y=-18, z=-7}
   }
   n6 = Anim Dynamics - Bone: Dyn_zipper!(SimulationSpace:AnimPhysSimSpaceType Enum, ComponentPose=@n5, Alpha=1, GravityOverride:Vector="0,0,-980", ExternalForce="0,0,0") -> (Pose) {
    angularDampingOverride=0.99
    angularSpringConstant=50
    bAngularSpring=true
    bOverrideAngularDamping=true
    bUseGravityOverride=true
    simulationSpace=RootRelative
    alphaBoolBlend {blendOption=Linear}
    boundBone {boneName=Dyn_zipper}
    componentAppliedLinearAccClamp=<3000 3000 2500>
    componentLinearAccScale=<0.5 0.5 0.5>
    componentLinearVelScale=<0.5 0.5 0.5>
    gravityOverride {z=-980}
    physicsBodyDefinitions [{
      boundBone {boneName=Dyn_zipper}
      boxExtents=<12 8 9>
      constraintSetup {coneAngle=1, angularLimitsMax=<5 15 10>, angularLimitsMin=<-5 -15 -30>, angularTarget=<0 0 0>}
      localJointOffset=<1 0 0>
     }]
   }
   n7 = Anim Dynamics - Bone: Dyn_zipper_01!(SimulationSpace:AnimPhysSimSpaceType Enum, ComponentPose=@n6, Alpha=1, GravityOverride:Vector="0,0,-980", ExternalForce="0,0,0") -> (Pose) {
    angularDampingOverride=0.99
    angularSpringConstant=50
    bAngularSpring=true
    bOverrideAngularDamping=true
    bUseGravityOverride=true
    simulationSpace=RootRelative
    alphaBoolBlend {blendOption=Linear}
    boundBone {boneName=Dyn_zipper_01}
    componentAppliedLinearAccClamp=<3000 3000 2500>
    componentLinearAccScale=<0.5 0.5 0.5>
    componentLinearVelScale=<0.5 0.5 0.5>
    gravityOverride {z=-980}
    physicsBodyDefinitions [{
      boundBone {boneName=Dyn_zipper_01}
      boxExtents=<12 8 10>
      constraintSetup {coneAngle=1, angularLimitsMax=<5 15 10>, angularLimitsMin=<-5 -10 -30>, angularTarget=<0 0 0>}
      localJointOffset=<1 0 0>
     }]
   }
   n8 = Anim Dynamics - Bone: Dyn_zipper_side!(SimulationSpace:AnimPhysSimSpaceType Enum, ComponentPose=@n7, Alpha=1, GravityOverride:Vector="0,0,-980", ExternalForce="0,0,0") -> (Pose) {
    angularDampingOverride=0.99
    angularSpringConstant=50
    bAngularSpring=true
    bOverrideAngularDamping=true
    bUseGravityOverride=true
    simulationSpace=RootRelative
    alphaBoolBlend {blendOption=Linear}
    boundBone {boneName=Dyn_zipper_side}
    componentAppliedLinearAccClamp=<3000 3000 2500>
    componentLinearAccScale=<0.5 0.5 0.5>
    componentLinearVelScale=<0.5 0.5 0.5>
    gravityOverride {z=-980}
    physicsBodyDefinitions [{
      boundBone {boneName=Dyn_zipper_side}
      boxExtents=<12 8 10>
      constraintSetup {coneAngle=1, angularLimitsMax=<5 3 10>, angularLimitsMin=<-5 -35 -10>, angularTarget=<0 0 0>}
      localJointOffset=<1 0 0>
     }]
   }
   n9 = Anim Dynamics - Bone: Dyn_Handle_01!(SimulationSpace:AnimPhysSimSpaceType Enum, ComponentPose=@n8, Alpha=1, GravityOverride:Vector="0,0,-980", ExternalForce="0,0,0") -> (Pose) {
    angularDampingOverride=0.99
    angularSpringConstant=50
    bAngularSpring=true
    bOverrideAngularDamping=true
    bUseGravityOverride=true
    simulationSpace=RootRelative
    alphaBoolBlend {blendOption=Linear}
    boundBone {boneName=Dyn_Handle_01}
    componentAppliedLinearAccClamp=<3000 3000 2500>
    componentLinearAccScale=<0.8 0.8 0.5>
    componentLinearVelScale=<0.8 0.8 0.5>
    gravityOverride {z=-980}
    physicsBodyDefinitions [{
      boundBone {boneName=Dyn_Handle_01}
      boxExtents=<12 12 12>
      constraintSetup {coneAngle=1, angularLimitsMax=<5 7 15>, angularLimitsMin=<-5 -7 -15>, angularTarget=<0 0 0>}
      localJointOffset=<-1 0 0.1>
     }]
   }
   n10 = "Constraint - Lable_corr (PP)"!(ComponentPose=@n9, Alpha=1, ConstraintWeights_0=0.2, ConstraintWeights_1:float=0.8) -> (Pose) {
    alphaBoolBlend {blendOption=Linear}
    boneToModify {boneName=Lable_corr}
    constraintSetup [{targetBone {boneName=pack}}, {targetBone {boneName=Base_Bat}}]
    constraintWeights [1,1]
   }
   Save cached pose 'InputPose'(Pose=Component To Local(ComponentPose=@n10)) {cacheName=InputPose}
  ]}
 EventGraph {nodes [
   Event BlueprintUpdateAnimation()
   > n1 = Sequence!(execute) -> (then_0, then_1, then_2:Exec, then_3:Exec)
   n2 = GetPartSkeletalMeshComponent()
   Is Valid(exec=@n1.then_1, InputObject=@n2)
   > Is Valid(InputObject=GetAttachParent!(self:Scene Component Object Reference=@n2) -> (ReturnValue:Scene Component Object Reference))
   n3 = GetDeltaTransformFromRefPose!(self=GetOwnerSkeletalMeshComponent(), BoneName=spine_05, BaseName=spine_01) -> (ReturnValue_Location:Vector, ReturnValue_Rotation:Rotator, ReturnValue_Scale:Vector)
   > Set Parashuting(Parashuting:double=MapRangeClamped(Value=BreakRotator(InRot=@n3.ReturnValue_Rotation).Yaw, InRangeA=-25, InRangeB=-60)) -> (Output_Get:double)
   n4 = Branch(execute=@n1.then_2, Condition=Get bIsInFrontEnd())
   > Set SimlationSpace(SimlationSpace:AnimPhysSimSpaceType Enum=Component) -> (Output_Get:AnimPhysSimSpaceType Enum)
   Set SimlationSpace(execute=@n4.else)
   n5 = Branch(
    execute=@n1.then_3
    Condition=PickaxeItemDefinitionHasAnyMetaTag(Tags=Make Array("[0]"=(TagName="Cosmetics.MetaTags.AstralLilac")), bRequiresEquipped=true)
   )
   > Set BatInHand(BatInHand:bool=false) -> (Output_Get:bool)
   Set BatInHand(execute=@n5.else, BatInHand=true)
   n6 = Gravity Override State Function DEPRECATED(
    execute=@n1.then_0
    CharacterPartAnimInstance=Self-Reference()
    GameTimeInSeconds=GetGameTimeInSeconds()
    GravityOverrideParams=Get Gravity Override Params()
    DebugJointName=spine_05
   )
   > Set Gravity Override(Gravity Override=@n6)
   > Set SkyState(SkyState:bool=Contains(SearchIn=@n6.CurrentStateOut, Substring=SkyDiving)) -> (Output_Get:bool)
   TryGetPawnOwner()
   CharacterItemDefinitionHasAnyMetaTag(Tags=Make Array("[0]"=(TagName="Cosmetics.MetaTags.AstralLilac")))
   PrintString!(execute, InString=Hello, bPrintToScreen, bPrintToLog, TextColor, Duration, Key) -> (then)
   Branch(Condition="Equal Exactly (String)"(B=Pickaxe_AstralLilac))
  ]}
]
variables [
 BatInHand {type=boolean}
 Gravity Override {type=Vector}
 Gravity Override Params {type=GravityOverrideParamsStruct}
 Mesh_Name {type=SkeletalMeshComponent}
 Parashuting {type=number}
 SimlationSpace {type=AnimPhysSimSpaceType}
 SkyState {type=boolean}
]
